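# MF_TwistUV is a material function that outputs twisted texture coordinates.
# A saturated SphereMask around the rotation center parameter, shaped by radius and hardness parameters, serves as the alpha that blends the rotation angle toward the angle parameter.
# That per-pixel angle feeds CustomRotator, so rotation varies with distance from the center, which is shared by the mask and the rotator.
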MaterialFunction /Game/Global/Materials/MaterialGym/Materials/MaterialElements/UVs/UVMainpulation/TwistUV/MF_TwistUV
graph {nodes [
  n1 = ScalarParameter() {defaultValue=0.5, group=TwistUV, parameterName=Twist_RotationCenter}
  n2 = ScalarParameter() {defaultValue=1, group=TwistUV, parameterName=Twist_RotationAngle}
  n3 = Saturate(None=SphereMask(
    A=TextureCoordinate()
    B=@n1
    Radius=ScalarParameter() {defaultValue=0.3748, group=TwistUV, parameterName=Twist_Radius}
    Hardness=ScalarParameter() {group=TwistUV, parameterName=Twist_Hardness}
   ))
  FunctionOutput(None=MaterialFunctionCall(UVs=TextureCoordinate(), Rotation Center=@n1, "Rotation Angle (0-1)"=LinearInterpolate(B=@n2, Alpha=@n3)) {materialFunction=&CustomRotator}.Rotated Values)
 ]}
