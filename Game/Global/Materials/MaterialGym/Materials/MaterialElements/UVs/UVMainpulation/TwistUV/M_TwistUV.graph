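# M_TwistUV is a Material Gym element material that demonstrates twisted UV manipulation on a Default Lit surface.
# The MF_TwistUV material function generates the UVs for a single texture parameter, UVTexture, which defaults to the UV editor color grid to make the distortion readable.
# The sample uses the shared world-group wrap sampler.
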Material /Game/Global/Materials/MaterialGym/Materials/MaterialElements/UVs/UVMainpulation/TwistUV/M_TwistUV
graph {
 properties {bAutomaticallySetUsageInEditor=false, usedShadingModels=Default Lit}
 nodes [TextureSampleParameter2D(UVs=MaterialFunctionCall() {materialFunction=&/MF_TwistUV}.Result) {group=TwistUV, parameterName=UVTexture, samplerSource=SSM_Wrap_WorldGroupSettings, texture=/UVEditor/Textures/UVEditorColorGrid_Color}]
}
